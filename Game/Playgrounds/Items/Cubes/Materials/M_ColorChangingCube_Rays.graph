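# Additive unlit material for the rays on a color-changing cube, which rise in an animated ring from an overlap point, with a TeamColor parameter.
# A timestamp function grows a sphere mask over quantized UV1 blocks, reshaped into a ring with per-block noise and sine variation.
# That mask drives vertex offset along local RiseOnZ with SplineThicken width, collapsing inactive vertices to the object position.
# It also scales a UV gradient that softens each ray.
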Material /Game/Playgrounds/Items/Cubes/Materials/M_ColorChangingCube_Rays
graph {
 properties {bEnableResponsiveAA=true, blendMode=BLEND_Additive, pixelDepthOffsetMode=PDOM_Legacy, shadingModel=MSM_Unlit, usedShadingModels=Unlit}
 nodes [
  n1 = TransformPosition(None=WorldPosition().XYZ) {transformSourceType=TRANSFORMPOSSOURCE_World}
  n2 = ObjectBounds()
  ComponentMask(None=Divide(A=Abs(None=@n1), B=Multiply(A=@n2) {constB=2})) {g=true, r=true}
  n3 = MaterialFunctionCall!(
   PrimaryTextureCoordinates=Multiply(A=TextureCoordinate() {coordinateIndex=1}, B=ScalarParameter() {defaultValue=30, parameterName=RingSteps})
   2ndOrderTextureCoordinates
   RandomSeed
   NoiseDensityBias
   Bypass Noise Influence Crosstalk
  ) {materialFunction=/Game/Effects/Fort_Effects/Materials/_Functions/MF_RandomUVSample}
  n4 = Subtract(A=@n3.NoiseWithRange) {constB=0.5}
  n5 = ScalarParameter() {defaultValue=0.05, parameterName=BlockDistortionScale}
  n6 = ScalarParameter() {defaultValue=30, parameterName=RingSteps}
  n7 = Floor(None=Multiply(A=TextureCoordinate() {coordinateIndex=1}, B=@n6))
  n8 = ComponentMask(None=VectorParameter() {parameterName=NormalizedOverlapPointLocalSpace, defaultValue=<rgba 0.5 0.5 0 0>}.RGB) {g=true, r=true}
  n9 = VectorParameter() {parameterName=CubeTimestampPayload, defaultValue=<rgba 0 0 0 1>}
  n10 = Multiply(A=Distance(A=@n8, B=Constant2Vector() {g=0.5, r=0.5})) {constB=1.4142}
  n11 = Multiply(
   A=MaterialFunctionCall(TimestampA=@n9.R, TransitionLength_In=@n9.R, TimestampB=@n9.R, TransitionLength_Out=@n9.R, Transition In Only?=StaticBool()) {materialFunction=&MF_Timestamp_InOut_01}.Linear
   B=LinearInterpolate(Alpha=Saturate(None=@n10)) {constA=1.75, constB=3.25}
  ) {constB=3.25}
  n12 = SphereMask(A=Add(A=Multiply(A=@n4, B=@n5), B=Divide(A=@n7, B=@n6)), B=@n8, Radius=@n11, Hardness=Constant())
  n13 = ScalarParameter() {defaultValue=0.3, parameterName=RingCenter}
  n14 = ScalarParameter() {defaultValue=0.2, parameterName=RingRadius}
  n15 = ScalarParameter() {defaultValue=0.2, parameterName=RingHardness}
  n16 = Add(A=Multiply(A=ScalarParameter() {defaultValue=2, parameterName=BlocksSinePeriodScale}, B=@n11), B=@n3.NoiseWithRange)
  n17 = Multiply(
   A=Multiply(A=@n12, B=SphereMask(A=@n12, B=@n13, Radius=@n14, Hardness=@n15))
   B=Add(A=Sine(None=@n16), B=ScalarParameter() {defaultValue=1.5, parameterName=BlockBias})
  )
  n18 = MaterialFunctionCall(Float2=TextureCoordinate()) {materialFunction=&BreakOutFloat2Components}
  n19 = Transform(None=Multiply(A=VectorParameter() {parameterName=RiseOnZ, defaultValue=<rgba 0 0 125 0>}.RGB, B=OneMinus(None=@n18.G))) {transformSourceType=TRANSFORMSOURCE_Local}
  n20 = ScalarParameter() {defaultValue=22, parameterName="WidthBase (S)"}
  n21 = Subtract(A=ObjectPositionWS(), B=WorldPosition().XYZ)
  If(
   A=@n17
   B=Constant() {r=0.005}
   "A > B"=Add(
    A=Multiply(A=@n17, B=@n19)
    B=MaterialFunctionCall(WidthBase=@n20, WidthTip=@n20) {materialFunction=&SplineThicken}.WorldPositionOffset
   )
   "A == B"=@n21
   "A < B"=@n21
  )
  n22 = MaterialFunctionCall(Float2=TextureCoordinate()) {materialFunction=&BreakOutFloat2Components}
  n23 = ComponentMask(None=@n22.G) {g=true}
  n24 = Constant() {r=0.5}
  n25 = SphereMask(A=@n22.R, B=@n24, Radius=@n24, Hardness=ScalarParameter() {defaultValue=0.15, parameterName=HorizontalGradientHardness})
  Multiply(A=Power(Base=Multiply(A=@n23, B=@n25), Exp=ScalarParameter() {defaultValue=4, parameterName=GradientExponent}), B=@n17)
  VectorParameter() {parameterName=TeamColor, defaultValue=<rgba 0.1859 0.1034 1 1>}
 ]
}
