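# Master material for Creative mode cube items, supporting Nanite and spline meshes, with plain parameters for color, metallic, specular, roughness, emissive and opacity mask.
# A destructible damage function drives World Position Offset for hit animation from a destruction amount parameter, with a static bool for hit effects that defaults to on.
# The combined material attributes then pass through the Creative move tool function.
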Material /Game/Playgrounds/Items/Cubes/Materials/M_Creative_Cube_Master
graph {
 properties {
  bUseMaterialAttributes=true
  bUsedWithNanite=true
  bUsedWithSplineMeshes=true
  ditherOpacityMask=true
  pixelDepthOffsetMode=PDOM_Legacy
  usedShadingModels=Default Lit
 }
 nodes [
  n1 = SetMaterialAttributes!(
   MaterialAttributes
   Base Color=VectorParameter() {parameterName=Color, defaultValue=<rgba 0.5 0.5 0.5 0>}.RGB
   Metallic=ScalarParameter() {parameterName=Metallic}
   Specular=ScalarParameter() {parameterName=Spec}
   Roughness=ScalarParameter() {parameterName=Roughness}
   Emissive Color=VectorParameter() {parameterName=Emissive, defaultValue=<rgba 0 0 0 1>}.RGB
   Opacity Mask=ScalarParameter() {defaultValue=1, parameterName=Opacity}
   World Position Offset=MaterialFunctionCall(
    UseDestructibleDamage=StaticBoolParameter() {defaultValue=true, group=z_GameplayParameters_DestructibleDamage, parameterName=Code_UseHitEffects}
    DestructibleDamageAmount=ScalarParameter() {group=z_GameplayParameters_DestructibleDamage, parameterName=Code_DestructionAmount}
   ) {materialFunction=&MF_FN_DestructibleDamage}.DamageAnimWPO
  ) {attributeSetTypes [
    #1
    #2
    #3
    #4
    #5
    #6
    #7
   ]}
  MaterialFunctionCall(MA In=@n1) {materialFunction=&MF_CreativeMoveTool}
 ]
}
