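# Playground cube material that wipes between two team colors from a local overlap point, with glowing transition effects.
# A timestamp transition, or debug override, grows a sphere mask over noise-jittered, stepped bounding-box XY coordinates, scaled by the point's distance from center.
# That mask blends the tint and drives emissive blocky rings, joined by sine-faded crack and edge glows.
# Local-space world-aligned road textures supply surface detail, and MF_CreativeMoveTool wraps the result.
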Material /Game/Playgrounds/Items/Cubes/Materials/M_Cube_ColorChanging
graph {
 properties {bUseMaterialAttributes=true, pixelDepthOffsetMode=PDOM_Legacy, usedShadingModels=Default Lit}
 nodes [
  n1 = VectorParameter() {parameterName=TeamColorA, defaultValue=<rgba 0.8 0.8 0.8 1>}
  n2 = VectorParameter() {parameterName=TeamColorB, defaultValue=<rgba 0.8 0.8 0.8 1>}
  n3 = ComponentMask(None=OneMinus(None=MaterialFunctionCall() {materialFunction=&BoundingBoxBased_0-1_UVW}.RGB)) {g=true, r=true}
  n4 = ScalarParameter() {defaultValue=30, parameterName=RingSteps}
  n5 = Multiply(A=@n3, B=@n4)
  n6 = MaterialFunctionCall!(PrimaryTextureCoordinates=@n5, 2ndOrderTextureCoordinates, RandomSeed, NoiseDensityBias, Bypass Noise Influence Crosstalk) {materialFunction=/Game/Effects/Fort_Effects/Materials/_Functions/MF_RandomUVSample}
  n7 = ScalarParameter() {defaultValue=0.5, parameterName=DebugTransitionValue}
  n8 = VectorParameter() {parameterName=CubeTimestampPayload, defaultValue=<rgba 0 0 0 1>}
  n9 = MaterialFunctionCall(TimestampA=@n8.R, TransitionLength_In=@n8.R, TimestampB=@n8.R, TransitionLength_Out=@n8.R, Transition In Only?=StaticBool()) {materialFunction=&MF_Timestamp_InOut_01}
  n10 = ComponentMask(None=VertexInterpolator(VS=StaticSwitchParameter(True=@n7, False=@n9.Linear) {parameterName=DebugTransition}).PS) {r=true}
  n11 = Add(A=Multiply(A=Subtract(A=@n6.NoiseWithRange) {constB=0.5}, B=Multiply(A=@n10) {constB=0.25}), B=Divide(A=Floor(None=@n5), B=@n4))
  n12 = ComponentMask(None=VectorParameter() {parameterName=NormalizedOverlapPointLocalSpace, defaultValue=<rgba 0.5 0.5 0 0>}.RGB) {g=true, r=true}
  n13 = Multiply(A=Distance(A=@n12, B=Constant2Vector() {g=0.5, r=0.5})) {constB=1.4142}
  n14 = Multiply(A=@n10, B=LinearInterpolate(Alpha=Saturate(None=@n13)) {constA=1.75, constB=3.25}) {constB=3.25, desc="3.25!"}
  n15 = Saturate(None=SphereMask(A=@n11, B=@n12, Radius=@n14, Hardness=Constant() {r=0.6}))
  n16 = MaterialFunctionCall!(
   WorldScale=VectorParameter() {parameterName="WorldScale (V3)", defaultValue=<rgba 512 512 512 0>}.RGB
   TransformToLocalSpace=StaticBool() {value=true}
  ) {materialFunction=/Game/Items/Traps/Materials/MF_Creative_WorldAlignedUVs}
  n17 = Desaturation(
   None=TextureSampleParameter2D(UVs=@n16.Result) {parameterName=Diffuse, texture=/Game/Packages/DS_Fortnight/SM/Textures/T_Street_Road_D}.RGB
   Fraction=Constant() {r=1}
  )
  n18 = LinearInterpolate(
   A=ScalarParameter() {defaultValue=-0.1, parameterName=DiffuseLow}
   B=ScalarParameter() {defaultValue=4, parameterName=DiffuseHigh}
   Alpha=@n17
  )
  n19 = TextureSampleParameter2D(UVs=@n16.Result) {parameterName=SpecTexture, samplerType=SAMPLERTYPE_Masks, texture=/Game/Packages/DS_Fortnight/SM/Textures/T_Street_Road_M}
  n20 = SphereMask(A=@n11, B=@n12, Radius=@n14, Hardness=Constant())
  n21 = SphereMask(
   A=@n20
   B=ScalarParameter() {defaultValue=0.3, parameterName=RingCenter}
   Radius=ScalarParameter() {defaultValue=0.2, parameterName=RingRadius}
   Hardness=ScalarParameter() {defaultValue=0.2, parameterName=RingHardness}
  )
  n22 = TextureSample(UVs=Divide(A=@n5) {constB=4}) {texture=/Game/Weapons/Prototype/ObjectMover/Textures/T_Movetool_Transition}
  n23 = Add(A=Multiply(A=ScalarParameter() {defaultValue=2, parameterName=BlocksSinePeriodScale}, B=@n14), B=@n6.NoiseWithRange)
  n24 = Multiply(
   A=Multiply(A=Multiply(A=@n20, B=@n21), B=OneMinus(None=@n22.R))
   B=Add(A=Sine(None=@n23), B=ScalarParameter() {defaultValue=1.5, parameterName=BlockBias})
  )
  n25 = LinearInterpolate(
   A=ScalarParameter() {defaultValue=1.1, parameterName=EmissiveCracksLow}
   B=ScalarParameter() {defaultValue=-11, parameterName=EmissiveCracksHigh}
   Alpha=@n17
  )
  n26 = ConstantBiasScale(None=@n3) {bias=-0.5, scale=2}
  n27 = SphereMask(
   A=Power(Base=Abs(None=@n26), Exp=ScalarParameter() {defaultValue=85, parameterName=EdgeGlowExponent}) {constExponent=10}
   B=Constant()
   Radius=Constant() {r=1}
   Hardness=ScalarParameter() {defaultValue=0.1, parameterName=EdgeGlowHardness}
  )
  n28 = Multiply(A=Sine(None=Multiply(A=@n10) {constB=0.5}), B=Add(A=Saturate(None=@n25), B=OneMinus(None=@n27)))
  n29 = SetMaterialAttributes!(
   MaterialAttributes
   Base Color=Multiply(A=LinearInterpolate(A=@n1.RGB, B=@n2.RGB, Alpha=@n15), B=Saturate(None=@n18))
   Metallic=@n19.G
   Specular=@n19.G
   Roughness=@n19.G
   Emissive Color=Multiply(A=Add(A=@n24, B=@n28), B=VectorParameter() {parameterName=EdgeColor, defaultValue=<rgba 1.9891 0.4108 2 0>}.RGB)
   Normal=TextureSampleParameter2D(UVs=@n16.Result) {parameterName=Normal, samplerType=SAMPLERTYPE_Normal, texture=/Game/Packages/DS_Fortnight/SM/Textures/T_Street_Road_N}.RGB
  ) {attributeSetTypes [
    #1
    #2
    #3
    #4
    #5
    #6
   ]}
  MaterialFunctionCall(MA In=@n29) {materialFunction=&MF_CreativeMoveTool}
 ]
}
