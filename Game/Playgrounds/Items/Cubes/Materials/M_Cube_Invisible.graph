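# M_Cube_Invisible is a Default Lit material for playground cube items, flagged for Nanite and instanced static meshes.
# Its graph holds a Subtract of world position from object position and a bright reddish vector parameter.
# Neither node connects to a material output, so despite the name the graph contains no logic that hides the cube.
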Material /Game/Playgrounds/Items/Cubes/Materials/M_Cube_Invisible
graph {
 properties {bUsedWithInstancedStaticMeshes=true, bUsedWithNanite=true, pixelDepthOffsetMode=PDOM_Legacy, usedShadingModels=Default Lit}
 nodes [
  Subtract(A=ObjectPositionWS(), B=WorldPosition().XYZ)
  VectorParameter() {defaultValue=<rgba 5 0.6017 0.5884 1>}
 ]
}
