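# Lit material for a Playgrounds disco cube: black panels with a grid of pulsing emissive lights, wrapped in the Fortnite base and Creative move-tool functions.
# UV2 tiles as blocks and sub-cells, optionally scaled by object radius, sampling a dance-floor texture flat and bump-offset to fake internal bulbs.
# Cell indices drive layered sine pulses and cycling patterns, which hue-shift the color when ColorShift is enabled (off by default).
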Material /Game/Playgrounds/Items/Cubes/Materials/M_PG_Cube_Disco
graph {
 properties {bUseMaterialAttributes=true, pixelDepthOffsetMode=PDOM_Legacy, refractionMethod=RM_IndexOfRefraction, usedShadingModels=Default Lit}
 nodes [
  n1 = Constant()
  n2 = ScalarParameter() {defaultValue=10, group=Panel Properties, parameterName=Metallic}
  n3 = ScalarParameter() {defaultValue=0.5, group=Panel Properties, parameterName=Specular}
  n4 = ScalarParameter() {group=Panel Properties, parameterName=Roughness}
  n5 = ScalarParameter() {defaultValue=822.4039, desc=Macro Iteration, parameterName=LargeBlock_Tiling}
  n6 = ObjectRadius()
  n7 = StaticSwitchParameter(
   True=Multiply(A=Divide(A=@n5, B=@n6), B=TextureCoordinate() {coordinateIndex=2}) {constB=2}
   False=Multiply(A=TextureCoordinate() {coordinateIndex=2}, B=Constant() {desc=Macro Iteration, r=2})
  ) {defaultValue=true, parameterName=PG_CubeSize}
  n8 = Constant() {desc=Micro Iteration, r=4}
  n9 = Multiply(A=@n7, B=@n8) {constB=4}
  n10 = Frac(None=@n9)
  n11 = TextureSample!(UVs=@n10, Tex, Level, Apply View MipBias) {
   mipValueMode=TMVM_MipLevel
   samplerSource=SSM_Clamp_WorldGroupSettings
   texture=/Game/Environments/Sets/Concert/Textures/DanceFloorBumpOffset
  }
  n12 = TextureSample(
   UVs=BumpOffset(Coordinate=@n10, Height=ScalarParameter() {defaultValue=-3.4312, group=Color, parameterName=Bump offset amount}) {referencePlane=1}
   Level=MaterialFunctionCall!(UVs=@n9, Texture Size=Constant() {r=256}) {materialFunction=/Engine/Functions/Engine_MaterialFunctions01/Texturing/ComputeMipLevel}.MipLevel
  ) {
   mipValueMode=TMVM_MipLevel
   samplerSource=SSM_Clamp_WorldGroupSettings
   texture=&DanceFloorBumpOffset
  }
  n13 = ScalarParameter() {defaultValue=200, group=Color, parameterName=Internal Light intensity}
  n14 = Add(
   A=Multiply(A=@n11.R, B=ScalarParameter() {defaultValue=8.1818, group=Color, parameterName=Surface Light Diffusion})
   B=Add(A=Multiply(A=@n12.G, B=@n13), B=ScalarParameter() {defaultValue=1, group=Color, parameterName=Internal light bias}) {constB=0.75}
  )
  n15 = Multiply(A=Power(Base=@n12.G) {constExponent=8}, B=ScalarParameter() {defaultValue=150, parameterName=Light Brightness}) {constB=600}
  n16 = ScalarParameter() {defaultValue=0.5, group=Color, parameterName=Light Bulb Color Shift Amount}
  n17 = MaterialFunctionCall(Float2=Ceil(None=@n7)) {materialFunction=&BreakOutFloat2Components}
  n18 = Multiply(A=@n17.G) {constB=2}
  n19 = Multiply(A=Subtract(A=Add(A=@n17.R, B=@n18)) {constB=3}) {constB=0.3333}
  n20 = Ceil(None=@n9)
  n21 = SphereMask(A=@n20, B=Add(A=@n8) {constB=0.5}, Radius=Add(A=@n8)) {attenuationRadius=5, hardnessPercent=0}
  n22 = ScalarParameter() {defaultValue=0.025, group=Patterns, parameterName=Color Animation Cycle Time}
  n23 = Constant() {r=1}
  n24 = MaterialFunctionCall(f2=@n20) {materialFunction=&AddComponents}
  n25 = Add(A=Time(), B=Divide(A=Fmod(A=@n24.f2, B=@n23), B=@n23))
  n26 = Multiply(A=Frac(None=Multiply(A=@n22, B=@n25) {constB=0.05, desc=Full cycle 20 second}) {desc="0-1 cycle "}) {constB=3, desc=the number of stages}
  n27 = ScalarParameter() {defaultValue=0.25, group=Patterns, parameterName=Pattern Transition Speed}
  n28 = MaterialFunctionCall(In=Clamp(None=@n26), Contrast=@n27) {materialFunction=&CheapContrast}
  n29 = Subtract(A=@n26)
  n30 = Clamp(None=@n29)
  n31 = Subtract(A=@n29)
  n32 = LinearInterpolate(
   A=LinearInterpolate(
    A=LinearInterpolate(A=@n19, B=@n21, Alpha=@n28.Result)
    B=Multiply(A=@n24.f2) {constB=0.125}
    Alpha=MaterialFunctionCall(In=@n30, Contrast=@n27) {materialFunction=&CheapContrast}.Result
   )
   B=@n19
   Alpha=MaterialFunctionCall(In=Clamp(None=@n31), Contrast=@n27) {materialFunction=&CheapContrast}.Result
  )
  n33 = Add(A=Multiply(A=ScalarParameter(), B=Time()), B=@n32)
  n34 = Add(A=@n7) {constB=0.5}
  n35 = Add(A=Multiply(A=Frac(None=@n34), B=@n8) {constB=4}) {constB=0.5}
  n36 = Abs(None=ConstantBiasScale(None=Frac(None=@n35)) {bias=-0.5, scale=3})
  n37 = Power(
   Base=MaterialFunctionCall!(f2=Clamp(None=@n36), f3, f4) {materialFunction=/Game/Packages/Fortnite_UniqueMaterials_1/VectorComponentOps}.f2 Mul
   Exp=Constant() {r=0.2571}
  )
  n38 = Multiply(A=@n37, B=ScalarParameter() {defaultValue=0.2, group=Color, parameterName=EdgeColorChange}) {constB=0.2}
  n39 = VectorParameter() {group=Color, parameterName=InitialColor, defaultValue=<rgba 0.974 0 0.8027 1>}
  n40 = MaterialFunctionCall(Hue Shift Percentage=Add(A=Multiply(A=@n12.G, B=@n16) {constB=10}, B=Add(A=@n33, B=@n38)), Texture=@n39.RGB) {materialFunction=&HueShift}
  n41 = Multiply(A=StaticSwitchParameter(True=Add(A=@n15, B=@n40.Result), False=@n39.RGB) {parameterName=ColorShift}, B=@n37) {desc=Soft Shadow}
  n42 = Add(A=Multiply(A=Time()) {constB=0.5}, B=@n32)
  n43 = Multiply(
   A=Add(A=Sine(None=@n42), B=ScalarParameter() {defaultValue=3, group=Pulse, parameterName=Animation Sine Pulse Low Value}) {constB=1.25}
   B=ScalarParameter() {defaultValue=0.01, group=Pulse, parameterName=Animation Sine Pulse High Value}
  ) {constB=0.5}
  n44 = MaterialFunctionCall(Float2=Ceil(None=@n20)) {materialFunction=&BreakOutFloat2Components}
  n45 = Subtract(A=Add(A=@n44.R, B=Multiply(A=@n44.G) {constB=2})) {constB=3}
  n46 = ScalarParameter() {defaultValue=2.6885, group=Pulse, parameterName=The Magic Number}
  n47 = Multiply(
   A=Add(A=Time(), B=Multiply(A=@n45, B=@n46) {constB=0.05})
   B=ScalarParameter() {defaultValue=0.5, group=Pulse, parameterName=Sine Pulse Speed}
  )
  n48 = Multiply(
   A=Add(A=Sine(None=@n47), B=ScalarParameter() {defaultValue=5.4534, group=Pulse, parameterName=Sine Pulse Low Value})
   B=ScalarParameter() {defaultValue=0.0595, group=Pulse, parameterName=Sine Pulse High Value}
  )
  n49 = Multiply(
   A=ScalarParameter() {defaultValue=5, group=Player Interaction, parameterName=Value multiplier}
   B=Multiply(A=Multiply(A=@n14, B=@n41) {desc=parallax}, B=Multiply(A=@n43, B=@n48))
  )
  MaterialFunctionCall(MA In=MaterialFunctionCall!(
    Material attributes=MakeMaterialAttributes!(
     BaseColor=@n1
     Metallic=@n2
     Specular=@n3
     Roughness=@n4
     Anisotropy
     EmissiveColor=@n49
     Opacity
     OpacityMask
     Normal
     Tangent
     WorldPositionOffset
     SubsurfaceColor
     ClearCoat
     ClearCoatRoughness
     AmbientOcclusion
     Refraction
     CustomizedUVs_0
     CustomizedUVs_1
     CustomizedUVs_2
     CustomizedUVs_3
     CustomizedUVs_4
     CustomizedUVs_5
     CustomizedUVs_6
     CustomizedUVs_7
     PixelDepthOffset
     ShadingModel
     Displacement
    )
    IsGameplayAsset
   ) {materialFunction=/Game/Packages/Fortress_SharedMaterials/Base_Material_Functions/Fortnite_BaseFunction}.Material Attributes) {materialFunction=&MF_CreativeMoveTool}
  MaterialFunctionCall() {materialFunction=&WorldAlignedTexture}
 ]
}
